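# Translucent unlit fire material for the Gimmick Male glider, built for Niagara mesh particles.
# Emissive blends smoke-distorted panning noise with a sphere-mask color gradient, tinted by particle color, with an optional simple override.
# Opacity layers three noise-warped samples of the glider mask, with dynamic parameters driving spike length, tail length and tail fade.
# Optional tail and detail-flame masks refine the shape; depth fade is skipped on forward and mobile shading paths.
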Material /BRCosmetics/Gadgets/Assets/VinderTech_GliderChute/Glider_Gimmick_Male/FX/M_Glider_Gimmick_Male_Fire
graph {
 properties {
  bUsedWithNiagaraMeshParticles=true
  blendMode=BLEND_Translucent
  pixelDepthOffsetMode=PDOM_Legacy
  shadingModel=MSM_Unlit
  usedShadingModels=Unlit
 }
 nodes [
  n1 = TextureCoordinate() {uTiling=2, vTiling=0.4}
  n2 = TextureSample(UVs=Panner(Coordinate=TextureCoordinate() {vTiling=0.5}) {speedY=0.2}) {texture=&T_Inky_Smoke_Tile}
  n3 = TextureSample(UVs=Add(A=Panner(Coordinate=@n1) {speedY=1}, B=Multiply(A=@n2.R) {constB=0.2})) {samplerType=SAMPLERTYPE_Grayscale, texture=&LowResBlurredNoise}
  n4 = Multiply(A=LinearInterpolate(
    B=VectorParameter() {group=Emissive, parameterName=BaseColor, defaultValue=<rgba 0.1719 0 0.1119 1>}.RGB
    Alpha=LinearInterpolate(Alpha=@n3.R) {constA=0.5, constB=0.8}
   )) {constB=10}
  n5 = TextureCoordinate()
  n6 = Constant2Vector() {g=0.4, r=0.5}
  n7 = LinearInterpolate(
   A=VectorParameter() {group=Emissive, parameterName=SphereMask_Color_Dark, defaultValue=<rgba 0.2969 0 0.1352 1>}.RGB
   B=VectorParameter() {group=Emissive, parameterName=SphereMask_Color_Birght, defaultValue=<rgba 3 0 0.5035 1>}.RGB
   Alpha=Power(Base=SphereMask(A=@n5, B=@n6) {attenuationRadius=0.5, hardnessPercent=0})
  ) {constB=3}
  n8 = Multiply(
   A=ParticleColor().RGB
   B=LinearInterpolate(A=Multiply(A=@n4, B=@n7), B=Constant() {r=1}, Alpha=ScalarParameter() {group=Emissive, parameterName=Use_Simple_Emissive})
  )
  MaterialFunctionCall(Emissive=@n8) {materialFunction=&MF_FX_EmissiveAdjustments}
  n9 = MaterialFunctionCall(UVs=TextureCoordinate(), "Rotation Angle (0-1)"=Constant() {r=-0.03}) {materialFunction=&CustomRotator}
  n10 = Constant2Vector() {g=0.7, r=2.5}
  n11 = TextureSample(UVs=Panner(Coordinate=Multiply(A=@n9.Rotated Values, B=@n10)) {speedX=-0.3, speedY=0.5}) {samplerType=SAMPLERTYPE_LinearGrayscale, texture=/Game/Global/Textures/Default/Noise/T_SwirlNoise_N_height}
  n12 = OneMinus(None=ComponentMask(None=TextureCoordinate()) {r=true})
  n13 = Constant() {r=0.5}
  n14 = Saturate(None=Multiply(A=SmoothStep(Min=@n12, Value=@n13)) {constB=5})
  n15 = Multiply(
   A=MaterialFunctionCall(UVs=TextureCoordinate(), "Rotation Angle (0-1)"=Constant() {r=0.03}) {materialFunction=&CustomRotator}.Rotated Values
   B=Constant2Vector() {g=0.7, r=2.5}
  )
  n16 = Multiply(A=TextureSample(UVs=Panner(Coordinate=@n15) {speedX=0.3, speedY=0.5}) {samplerType=SAMPLERTYPE_LinearGrayscale, texture=&T_SwirlNoise_N_height}.R) {constB=0.1}
  n17 = Add(A=ComponentMask(None=TextureCoordinate()) {r=true}, B=Multiply(A=LinearInterpolate(A=@n16, B=@n11.R, Alpha=@n14)) {constB=0.03}) {constB=0}
  n18 = TextureSample(UVs=Panner(Coordinate=TextureCoordinate() {uTiling=4, vTiling=0.5}) {speedX=0.3, speedY=1}) {texture=&T_Inky_Smoke_Tile}
  n19 = DynamicParameter() {paramNames [Spike length, Tail length, Tail Fade, Param4]}
  n20 = TextureCoordinate()
  n21 = Multiply(A=SmoothStep(Min=ComponentMask(None=@n20) {g=true}) {constValue=0.6}) {constB=10}
  n22 = Multiply(A=Multiply(A=Multiply(A=@n18.R, B=@n19.Spike length)) {constB=0.8}, B=Saturate(None=Saturate(None=@n21)))
  n23 = Add(A=ComponentMask(None=TextureCoordinate()) {g=true}, B=Multiply(A=LinearInterpolate(Alpha=@n22) {constA=-1, constB=0.5}) {constB=0.3})
  n24 = TextureSample(UVs=Add(A=AppendVector(A=@n17, B=@n23), B=Constant2Vector() {g=0.15, r=-0.01})) {
   samplerSource=SSM_Clamp_WorldGroupSettings
   texture=&/T_Gimmick_Glider_Mask
  }
  n25 = Subtract(A=ComponentMask(None=TextureCoordinate()) {g=true}) {constB=0.3}
  n26 = LinearInterpolate(A=Constant() {r=1}, B=Saturate(None=Multiply(A=@n25) {constB=5}), Alpha=ScalarParameter() {parameterName=Use_Tail_Fade})
  n27 = MaterialFunctionCall(UVs=TextureCoordinate(), Texture Scale=Constant() {r=1}) {materialFunction=&ScaleUVsByCenter}
  n28 = MaterialFunctionCall(UVs=TextureCoordinate(), Texture Scale=Constant() {r=1.05}) {materialFunction=&ScaleUVsByCenter}
  n29 = ComponentMask(None=TextureCoordinate()) {g=true}
  n30 = TextureCoordinate()
  n31 = Constant() {r=0.03}
  n32 = Panner(Coordinate=Multiply(
    A=MaterialFunctionCall(UVs=@n30, "Rotation Angle (0-1)"=@n31) {materialFunction=&CustomRotator}.Rotated Values
    B=Constant2Vector() {g=0.7, r=2.5}
   )) {speedX=0.3, speedY=0.5}
  n33 = Add(
   A=LinearInterpolate(A=ComponentMask(None=@n27.UVs) {r=true}, B=ComponentMask(None=@n28.UVs) {r=true}, Alpha=Saturate(None=@n29))
   B=Multiply(A=TextureSample(UVs=@n32) {samplerType=SAMPLERTYPE_LinearGrayscale, texture=&T_SwirlNoise_N_height}.R) {constB=0.04}
  ) {constB=0}
  n34 = TextureSample(UVs=Panner(Coordinate=TextureCoordinate() {uTiling=1.5, vTiling=-0.5}) {speedX=0.3, speedY=0.5}) {samplerType=SAMPLERTYPE_Grayscale, texture=&LowResBlurredNoise}
  n35 = Add(A=ComponentMask(None=TextureCoordinate()) {g=true}, B=Multiply(A=Multiply(A=@n34.R) {constB=0.09}) {constB=-1}) {constB=0}
  n36 = TextureSample(UVs=Add(A=AppendVector(A=@n33, B=@n35), B=Constant2Vector() {g=-0.1, r=-0.02})) {
   samplerSource=SSM_Clamp_WorldGroupSettings
   texture=&T_Gimmick_Glider_Mask
  }
  n37 = TextureCoordinate()
  n38 = TextureCoordinate() {uTiling=3, vTiling=0.5}
  n39 = Multiply(
   A=TextureSample(UVs=Panner(Coordinate=@n38) {speedX=0.3, speedY=1}) {texture=&T_Inky_Smoke_Tile}.R
   B=DynamicParameter() {paramNames [Spike length, Tail length, Tail Fade, Param4]}.Tail length
  )
  n40 = ComponentMask(None=TextureCoordinate()) {r=true}
  n41 = Power(Base=OneMinus(None=Sine(None=@n40) {period=2}))
  n42 = LinearInterpolate(Alpha=Saturate(None=Multiply(A=@n41) {constB=5})) {constA=0.7, constB=2}
  n43 = ComponentMask(None=TextureCoordinate()) {g=true}
  n44 = Saturate(None=Multiply(A=SmoothStep(Min=@n43) {constValue=0.6}) {constB=10})
  n45 = LinearInterpolate(Alpha=Multiply(A=Multiply(A=@n39, B=@n42), B=Saturate(None=@n44))) {constA=-1, constB=0.5}
  n46 = AppendVector(A=ComponentMask(None=TextureCoordinate()) {r=true}, B=Add(A=ComponentMask(None=@n37) {g=true}, B=Multiply(A=@n45) {constB=0.3}))
  n47 = Constant2Vector() {g=0.4}
  n48 = TextureSample(UVs=MaterialFunctionCall(UVs=Add(A=@n46, B=@n47), Texture Scale=Constant2Vector() {g=1, r=0.9}) {materialFunction=&ScaleUVsByCenter}.UVs) {
   samplerSource=SSM_Clamp_WorldGroupSettings
   texture=&T_Gimmick_Glider_Mask
  }
  n49 = Sine(None=ComponentMask(None=TextureCoordinate()) {r=true}) {period=2}
  n50 = Multiply(A=Subtract(A=OneMinus(None=@n49)) {constB=0.13}) {constB=5}
  n51 = ComponentMask(None=TextureCoordinate()) {g=true}
  n52 = Multiply(A=Power(Base=OneMinus(None=@n51)) {constExponent=5}) {constB=10}
  n53 = Multiply(A=Saturate(None=Multiply(A=@n48.G) {constB=5}), B=Multiply(A=Saturate(None=@n50), B=Saturate(None=@n52)))
  n54 = Add(
   A=Add(A=@n24.R, B=Multiply(A=@n36.R, B=@n26)) {constB=0}
   B=Multiply(A=Saturate(None=@n53), B=DynamicParameter() {paramNames [Spike length, Tail length, Tail Fade, Param4]}.Tail Fade)
  ) {constB=0}
  n55 = Constant() {r=1}
  n56 = OneMinus(None=ComponentMask(None=TextureCoordinate()) {g=true})
  n57 = Saturate(None=Multiply(A=Power(Base=@n56) {constExponent=5}) {constB=10})
  n58 = ScalarParameter() {group=Opacity, parameterName=Use_Tail_mask}
  n59 = LinearInterpolate(Alpha=Multiply(A=Saturate(None=@n54), B=LinearInterpolate(A=@n55, B=@n57, Alpha=@n58))) {constA=-0.2}
  n60 = Multiply(A=TextureCoordinate())
  n61 = Add(A=OneMinus(None=ComponentMask(None=@n60) {g=true})) {constB=-0.3}
  n62 = ComponentMask(None=VectorParameter() {group=Detail flame mask, parameterName=UV_Scale, defaultValue=<rgba 1 1 0 0>}.RGB) {g=true, r=true}
  n63 = Multiply(A=@n62, B=TextureCoordinate() {uTiling=1.5, vTiling=0.6})
  n64 = TextureCoordinate() {uTiling=1.5, vTiling=0.3}
  n65 = Multiply(A=TextureSample(UVs=Panner(Coordinate=@n64) {speedY=0.1}) {texture=&T_Inky_Smoke_Tile}.R) {constB=0.2}
  n66 = TextureSample(UVs=Add(A=Panner(Coordinate=@n63) {speedY=1}, B=@n65)) {texture=&T_Inky_Smoke_Tile}
  n67 = Multiply(A=@n62, B=TextureCoordinate() {uTiling=1.5, vTiling=0.5})
  n68 = TextureSample(UVs=Add(A=@n65, B=Panner(Coordinate=@n67) {speedY=0.3})) {samplerType=SAMPLERTYPE_LinearColor, texture=/Game/Effects/Fort_Effects/BluGlow/Violentium/Violentium_Mask}
  n69 = LinearInterpolate(Alpha=Multiply(A=Multiply(A=@n66.R, B=@n68.R), B=ScalarParameter() {defaultValue=2, group=Detail flame mask, parameterName=Dark flame range}) {constB=2})
  n70 = LinearInterpolate(A=Round(None=@n69), B=@n69, Alpha=ScalarParameter() {defaultValue=0.8, group=Detail flame mask, parameterName=Dark flame Round}) {constAlpha=0.8}
  n71 = Multiply(
   A=Multiply(A=ParticleColor().A, B=Saturate(None=@n59))
   B=LinearInterpolate(A=Constant() {r=1}, B=Multiply(A=@n61, B=@n70), Alpha=ScalarParameter() {group=Detail flame mask, parameterName=Use_Detail_Mask})
  )
  n72 = DepthFade(Opacity=@n71) {fadeDistanceDefault=5}
  Reroute(None=ShadingPathSwitch(Default=@n72, Deferred=@n72, Forward=@n71, Mobile=@n71))
  Reroute(None=@n8)
 ]
}
